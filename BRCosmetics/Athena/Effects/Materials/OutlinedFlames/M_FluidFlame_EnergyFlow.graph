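# Translucent unlit flame material for Niagara meshes and ribbons, drawing outlined energy-flow flames tinted by particle color.
# Default shading fills them with distorted, screen-space scrolling noise; Forward and Mobile use flat color with black strokes from a secondary texture.
# Opacity combines a noise-warped panning mask, hard-rounded dissolve noise along U, edge falloffs and depth fade.
# Dynamic parameters drive offsets, pan speed and optional erosion.
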Material /BRCosmetics/Athena/Effects/Materials/OutlinedFlames/M_FluidFlame_EnergyFlow
graph {
 properties {
  allowTranslucentCustomDepthWrites=true
  bUsedWithNiagaraMeshParticles=true
  bUsedWithNiagaraRibbons=true
  blendMode=BLEND_Translucent
  opacityMaskClipValue=0.001
  pixelDepthOffsetMode=PDOM_Legacy
  shadingModel=MSM_Unlit
  twoSided=true
  usedShadingModels=Unlit
 }
 nodes [
  n1 = Constant2Vector() {g=1, r=6}
  n2 = ScreenPosition()
  n3 = TextureSample(UVs=Panner(Coordinate=Multiply(A=@n2.ViewportUV, B=@n1)) {speedX=-0.05}) {samplerType=SAMPLERTYPE_LinearGrayscale, texture=/Engine/Functions/Engine_MaterialFunctions02/ExampleContent/Textures/LowResBlurredNoise}
  n4 = Add(A=Multiply(A=ScreenPosition().ViewportUV, B=@n1), B=Multiply(A=@n3.R) {constB=0.1})
  n5 = ComponentMask(None=VectorParameter() {parameterName=Emissive Noise Speed, defaultValue=<rgba 0 0.5 0 0>}.RGB) {g=true, r=true}
  n6 = ParticleColor()
  n7 = Multiply(
   A=TextureSample(UVs=Panner(Coordinate=@n4, Speed=@n5) {speedX=-0.3}) {samplerType=SAMPLERTYPE_Masks, texture=/Game/Weapons/WeaponSkins/Wraps/Textures/Mash/Mash_Noise}.R
   B=@n6.RGB
  )
  n8 = Multiply(
   A=TextureCoordinate() {uTiling=1.2}
   B=ComponentMask(None=VectorParameter() {parameterName=Mask UV Scale, defaultValue=<rgba 1 1 0 0>}.RGB) {g=true, r=true}
  )
  n9 = DynamicParameter() {defaultValue=<rgba 0 0.5 0 1>, paramNames [U offset, panner, V offset, dissolve offset]}
  n10 = AppendVector(A=@n9.U offset, B=@n9.U offset)
  n11 = TextureSample(UVs=Panner(Coordinate=TextureCoordinate()) {speedX=-0.3, speedY=0.3}) {samplerType=SAMPLERTYPE_Grayscale, texture=/Game/Global/Textures/Default/Noise/LowResBlurredNoise}
  n12 = ScalarParameter() {defaultValue=0.2, parameterName=Noise intensity}
  n13 = Panner(
   Coordinate=Add(A=Add(A=@n8, B=@n10), B=Multiply(A=@n11.R, B=@n12) {constB=0.2})
   Speed=Multiply(
    A=Constant2Vector() {r=-1}
    B=DynamicParameter() {defaultValue=<rgba 0 0.5 0 1>, paramNames [U offset, panner, V offset, dissolve offset]}.panner
   )
  ) {speedX=-1}
  n14 = StaticComponentMaskParameter(None=TextureSampleParameter2D(UVs=@n13) {parameterName=Mask Tex, texture=/BRCosmetics/Athena/Effects/Textures/OutlinedFlames/T_FluidFlame_Mask}.RGB) {defaultR=true, parameterName=RGB mask}
  n15 = OneMinus(None=@n14)
  n16 = TextureSampleParameter2D(UVs=@n13) {group=LowPlatform, parameterName=LowPlatform Tex, samplerType=SAMPLERTYPE_Normal, texture=/Game/Effects/Fort_Effects/Textures/T_FluidFlame_N}
  n17 = MaterialFunctionCall(Float2=ConstantBiasScale(None=AppendVector(A=@n16.R, B=@n16.R))) {materialFunction=&BreakOutFloat2Components}
  n18 = ScalarParameter() {defaultValue=0.4, group=LowPlatform, parameterName=LowPlatform_Stroke_MaxSize}
  n19 = Saturate(None=Subtract(A=Saturate(None=@n17.R), B=@n18))
  n20 = OneMinus(None=@n18)
  n21 = Saturate(None=Multiply(A=Divide(A=@n19, B=@n20), B=ScalarParameter() {defaultValue=10, group=LowPlatform, parameterName=LowPlatform_Stroke_Hardness}))
  n22 = LinearInterpolate(A=@n6.RGB, Alpha=Saturate(None=Add(A=@n15, B=@n21))) {constB=0}
  MaterialFunctionCall(Emissive=ShadingPathSwitch(Default=@n7, Deferred=@n7, Forward=@n22, Mobile=@n22)) {materialFunction=&MF_FX_EmissiveAdjustments}
  n23 = Multiply(
   A=TextureCoordinate() {uTiling=2, vTiling=1.5}
   B=ComponentMask(None=VectorParameter() {parameterName=Dis UV Scale, defaultValue=<rgba 1 1 0 1>}.RGB) {g=true, r=true}
  )
  n24 = AppendVector(A=DynamicParameter() {defaultValue=<rgba 0 0.5 0 1>, paramNames [U offset, panner, V offset, dissolve offset]}.dissolve offset, B=Constant())
  n25 = VectorParameter() {parameterName=Dis UV Offset, defaultValue=<rgba 0 0.2 0 1>}
  n26 = Panner(
   Coordinate=Add(A=Add(A=@n23, B=@n24), B=ComponentMask(None=@n25.RGB) {g=true, r=true})
   Speed=ComponentMask(None=VectorParameter() {parameterName=Dis Speed, defaultValue=<rgba -1 0 0 0>}.RGB) {g=true, r=true}
  ) {speedX=-1}
  n27 = LinearInterpolate(
   A=ScalarParameter() {defaultValue=0.2, parameterName=Dis Min}
   B=DynamicParameter() {parameterIndex=1, defaultValue=<rgba 0 0.5 0 1>, paramNames [Erode,Param2,Param3,Param4]}.Erode
   Alpha=ScalarParameter() {parameterName=Use Dynamic Parameter Erode}
  )
  n28 = ScalarParameter() {defaultValue=2, parameterName=Dis Max}
  n29 = TextureCoordinate()
  n30 = Saturate(None=Add(A=ComponentMask(None=@n29) {r=true}) {constB=-0.5})
  n31 = Round(None=Power(
    Base=TextureSampleParameter2D(UVs=@n26) {parameterName=Dis_Tex, samplerType=SAMPLERTYPE_Grayscale, texture=&LowResBlurredNoise}.R
    Exp=LinearInterpolate(A=@n27, B=@n28, Alpha=@n30)
   ))
  n32 = OneMinus(None=ComponentMask(None=TextureCoordinate()) {r=true})
  n33 = Saturate(None=Multiply(A=Power(Base=@n32)) {constB=10})
  n34 = Multiply(A=Multiply(A=Multiply(A=@n14, B=@n31)), B=@n33)
  n35 = SphereMask(A=TextureCoordinate(), B=Constant2Vector() {g=0.5, r=0.6}) {attenuationRadius=0.6, hardnessPercent=0}
  n36 = Saturate(None=Multiply(A=Power(Base=@n35), B=ScalarParameter() {defaultValue=50, parameterName=Edge clear}) {constB=50})
  n37 = Multiply(A=Multiply(A=Multiply(A=@n34), B=@n36))
  n38 = Multiply(A=@n21, B=Multiply(A=Multiply(A=@n31, B=@n33), B=@n36))
  n39 = VectorParameter() {group=LowPlatform, parameterName=LowPlatform_OpacityRefine, defaultValue=<rgba 5 4 1 0.5>}
  n40 = ScalarParameter() {defaultValue=0.05, group=LowPlatform, parameterName=LowPlatform_BaseOpacity}
  n41 = Multiply(
   A=Add(
    A=MaterialFunctionCall(Texture=@n17.G, RGB=@n39.RGB, A=@n39.RGB) {materialFunction=&MF_FX_Refine}.Result
    B=Multiply(A=@n14, B=@n40)
   ) {constB=0}
   B=@n37
  )
  n42 = Multiply(A=Saturate(None=Add(A=@n38, B=@n41)), B=ScalarParameter() {defaultValue=10, parameterName=E3Opacity})
  n43 = ShadingPathSwitch(Default=@n37, Deferred=@n37, Forward=@n42, Mobile=@n42)
  Reroute(None=DepthFade(Opacity=Multiply(A=@n6.A, B=@n43), FadeDistance=ScalarParameter() {defaultValue=10, parameterName=FadeDistance}))
 ]
}
